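# Editor utility widget that pairs a labeled button with a text box for sending a configurable console command.
# PreConstruct applies label and sizes and fills the box with the console variable's float value.
# Clicking broadcasts OnClicked; with float state set, it executes command plus typed value without world context, refreshes the box and broadcasts OnPostClicked.
# The non-float and PIE-world-targeted execution nodes have no incoming exec link.
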EditorUtilityWidgetBlueprint /FNAnimUtilities/Widgets/NativizationConsoleCommandField
graphs [
 EventGraph {nodes [
   n1 = Get EditorUtilityEditableTextBox() -> (EditorUtilityEditableTextBox:Editor Utility Editable Text Box Object Reference)
   n2 = Append!(
    A=Get Console Command() -> (Console Command:String)
    B=" "
    C:String="To String (Text)"(InText="GetText (Text Box)"!(self:Text Box Object Reference=@n1) -> (ReturnValue:Text))
   ) -> (ReturnValue)
   n3 = Get Run On Server() -> (Run On Server:bool)
   n4 = GetPIEWorlds!(execute=@n8.then, bIncludeDedicatedServer:bool=@n3) -> (then, ReturnValue:Array of World Object References)
   n5 = ExecuteConsoleCommand(WorldContextObject="Get (a copy)"(Array:Array of World Object References=@n4) -> (Output:World Object Reference), Command=@n2)
   UpdateTextField(execute=[@n5.then, @n12.then])
   > Call On Post Clicked!(execute:Exec, self:Nativization Console Command Field Object Reference) -> (then:Exec) {is=title K2Node_CallDelegate}
   n6 = GetPIEWorlds(bIncludeDedicatedServer=@n3)
   > n7 = Branch(Condition=Is Not Empty!(TargetArray:Array of World Object References=@n6) -> (ReturnValue:bool))
   > ExecuteConsoleCommand(WorldContextObject="Get (a copy)"(Array=@n6), Command=Get Console Command())
   > Call On Post Clicked()
   ExecuteConsoleCommand(execute=@n7.else, Command=Get Console Command())
   "On Clicked (Button)"() {is=title K2Node_ComponentBoundEvent}
   > Call On Clicked!(execute:Exec, self:Nativization Console Command Field Object Reference) -> (then:Exec) {is=title K2Node_CallDelegate}
   > n8 = Branch(Condition=Get Command Has Float State() -> (Command Has Float State:bool))
   Event PreConstruct()
   > UpdateTextField()
   > "SetText (Text)"(self=Get Text!() -> (Text:Text Object Reference), InText=Get ButtonName() -> (ButtonName:Text))
   n9 = Get ButtonSizeBox() -> (ButtonSizeBox:Size Box Object Reference)
   > SetHeightOverride(self=@n9, InHeightOverride=Get Button Height() -> (Button Height:double))
   > SetWidthOverride(self=@n9, InWidthOverride=Get Button Width() -> (Button Width:double))
   n10 = Get TextBoxSizeBox() -> (TextBoxSizeBox:Size Box Object Reference)
   > SetHeightOverride(self=@n10, InHeightOverride=Get TextBox Height() -> (TextBox Height:double))
   > SetWidthOverride(self=@n10, InWidthOverride=Get TextBox Width() -> (TextBox Width:double))
   n11 = Is Not Empty(TargetArray=@n4)
   n12 = ExecuteConsoleCommand(execute=[@n13.else, @n13.then], Command=@n2)
   n13 = Branch(execute=@n4.then, Condition=@n11)
  ]}
 OnClicked {nodes [OnClicked() {is=title K2Node_FunctionEntry}]}
 OnPostClicked {nodes [OnPostClicked() {is=title K2Node_FunctionEntry}]}
 UpdateTextField {nodes [
   UpdateTextField() {is=title K2Node_FunctionEntry}
   > n1 = GetConsoleVariableFloatValue!(execute, VariableName:String=Get Console Command()) -> (then, ReturnValue:float)
   > "SetText (Text Box)"!(
    execute
    self:Text Box Object Reference=Get EditorUtilityEditableTextBox()
    InText:Text="To Text (Float)"(Value=@n1, MinimumIntegralDigits=1, MaximumIntegralDigits=324, MinimumFractionalDigits=0, MaximumFractionalDigits=6)
   ) -> (then)
  ]}
]
variables [
 Button Height {type=number, default=25}
 Button Width {type=number, default=75}
 ButtonColor {type=SlateColor, default {colorUseRule=UseColor_Specified, specifiedColor=<rgba 0.0009 1 0 1>}}
 ButtonName {type=string, default=Button Name}
 Command Has Float State {type=boolean}
 Console Command {type=string, default=""}
 OnClicked {default=()}
 OnPostClicked {default=()}
 Run On Server {type=boolean, default=true}
 TextBox Height {type=number, default=25}
 TextBox Width {type=number, default=75}
]
widgets {
 info {rootWidgetClass=&HorizontalBox, widgetCount=6}
 widgets [
  HorizontalBox_792 {widgetClassPath=&HorizontalBox, widget=HorizontalBox_792}
  ButtonSizeBox {
   widgetClassPath=&SizeBox
   bIsVariable=true
   parent=HorizontalBox_792
   widget=ButtonSizeBox
   slot {class=HorizontalBoxSlot, properties {padding=<ltrb 0 0 5 0>}}
  }
  Button {
   widgetClassPath=&EditorUtilityButton
   parent=ButtonSizeBox
   widget=Button
   properties {widgetStyle {
     clickedSlateSound {resourceObject=null}
     disabled {
      bIsDynamicallyLoaded=false
      drawAs=Box
      imageType=NoImage
      mirroring=NoMirror
      resourceName=None
      resourceObject=null
      tiling=NoTile
      imageSize=<0 0>
      margin=<ltrb 0 0 0 0>
      outlineSettings {
       bUseBrushTransparency=false
       roundingType=FixedRadius
       width=1
       color {colorUseRule=UseColor_ColorTable, specifiedColor=<rgba 0 0 0 0>}
       cornerRadii=<4 4 4 4>
      }
      tintColor {colorUseRule=UseColor_Specified, specifiedColor=<rgba 0.0176 0.0176 0.0176 1>}
      uVRegion {bIsValid=false, max=<0 0>, min=<0 0>}
     }
     disabledForeground {colorUseRule=UseColor_ColorTable, specifiedColor=<rgba 0 0 0 0>}
     hovered {
      bIsDynamicallyLoaded=false
      drawAs=RoundedBox
      imageType=NoImage
      mirroring=NoMirror
      resourceName=None
      resourceObject=null
      tiling=NoTile
      imageSize=<0 0>
      margin=<ltrb 0 0 0 0>
      outlineSettings {
       bUseBrushTransparency=true
       roundingType=FixedRadius
       width=1
       color {colorUseRule=UseColor_ColorTable, specifiedColor=<rgba 0 0 0 0>}
       cornerRadii=<4 4 4 4>
      }
      tintColor {colorUseRule=UseColor_Specified, specifiedColor=<rgba 0.6087 0.6087 0.6087 1>}
      uVRegion {bIsValid=false, max=<0 0>, min=<0 0>}
     }
     hoveredForeground {colorUseRule=UseColor_ColorTable, specifiedColor=<rgba 0 0 0 0>}
     hoveredSlateSound {resourceObject=null}
     normal {
      bIsDynamicallyLoaded=false
      drawAs=RoundedBox
      imageType=NoImage
      mirroring=NoMirror
      resourceName=None
      resourceObject=null
      tiling=NoTile
      imageSize=<0 0>
      margin=<ltrb 0 0 0 0>
      outlineSettings {
       bUseBrushTransparency=true
       roundingType=FixedRadius
       width=1
       color {colorUseRule=UseColor_ColorTable, specifiedColor=<rgba 0 0 0 0>}
       cornerRadii=<4 4 4 4>
      }
      tintColor {colorUseRule=UseColor_Specified, specifiedColor=<rgba 0.087 0.087 0.087 1>}
      uVRegion {bIsValid=false, max=<0 0>, min=<0 0>}
     }
     normalForeground {colorUseRule=UseColor_ColorTable, specifiedColor=<rgba 0 0 0 0>}
     normalPadding=<ltrb 0 0 0 0>
     pressed {
      bIsDynamicallyLoaded=false
      drawAs=Box
      imageType=NoImage
      mirroring=NoMirror
      resourceName=None
      resourceObject=null
      tiling=NoTile
      imageSize=<0 0>
      margin=<ltrb 0 0 0 0>
      outlineSettings {
       bUseBrushTransparency=true
       roundingType=FixedRadius
       width=1
       color {colorUseRule=UseColor_ColorTable, specifiedColor=<rgba 0 0 0 0>}
       cornerRadii=<4 4 4 4>
      }
      tintColor {colorUseRule=UseColor_Specified, specifiedColor=<rgba 0.0075 0.0075 0.0075 1>}
      uVRegion {bIsValid=false, max=<0 0>, min=<0 0>}
     }
     pressedForeground {colorUseRule=UseColor_ColorTable, specifiedColor=<rgba 0 0 0 0>}
     pressedPadding=<ltrb 0 0 0 0>
     pressedSlateSound {resourceObject=null}
    }}
   slot {class=SizeBoxSlot}
  }
  Text {
   widgetClassPath=&TextBlock
   bIsVariable=true
   parent=Button
   widget=Text
   properties {
    text=Button Name
    colorAndOpacity {colorUseRule=UseColor_Specified, specifiedColor=<rgba 0.9763 1 0.918 1>}
    font {
     bForceMonospaced=false
     bMaterialIsStencil=false
     fontMaterial=null
     fontObject=&Roboto
     letterSpacing=0
     monospacedWidth=1
     size=11.25
     skewAmount=0
     typefaceFontName=Regular
     outlineSettings {
      bApplyOutlineToDropShadows=false
      bMiteredCorners=false
      bSeparateFillAlpha=false
      outlineMaterial=null
      outlineSize=0
      outlineColor=<rgba 0 0 0 1>
     }
    }
    strikeBrush {
     bIsDynamicallyLoaded=false
     drawAs=Image
     imageType=NoImage
     mirroring=NoMirror
     resourceName=None
     resourceObject=null
     tiling=NoTile
     imageSize=<32 32>
     margin=<ltrb 0 0 0 0>
     outlineSettings {
      bUseBrushTransparency=false
      roundingType=HalfHeightRadius
      width=0
      color {colorUseRule=UseColor_Specified, specifiedColor=<rgba 0 0 0 0>}
      cornerRadii=<0 0 0 0>
     }
     tintColor {colorUseRule=UseColor_Foreground, specifiedColor=<rgba 1 1 1 1>}
     uVRegion {bIsValid=false, max=<0 0>, min=<0 0>}
    }
   }
   slot {class=ButtonSlot}
  }
  TextBoxSizeBox {
   widgetClassPath=&SizeBox
   bIsVariable=true
   parent=HorizontalBox_792
   widget=TextBoxSizeBox
   slot {class=HorizontalBoxSlot, properties {padding=<ltrb 5 0 0 0>}}
  }
  EditorUtilityEditableTextBox {
   widgetClassPath=&EditorUtilityEditableTextBox
   bIsVariable=true
   parent=TextBoxSizeBox
   widget=EditorUtilityEditableTextBox
   slot {class=SizeBoxSlot}
  }
 ]
}
